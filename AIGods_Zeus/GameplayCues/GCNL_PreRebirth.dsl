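; Looping gameplay cue notify for Zeus's pre-rebirth state that plays the boss shield start sound.
; On looping start it casts the target to BP_PlayerPawn_Zeus, waits briefly, then spawns the sound attached to that pawn's root component.
; The graph adds no visual effects and no logic for other cue events.
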(event EventOnLoopingStart(EitherVFX-preferthisversion)
  (params (MyTarget :type Actor%20Object%20Reference) (Parameters :type Gameplay%20Cue%20Parameters%20Structure%20%28by%20ref%29) (ParticleComponents :type Array%20of%20FXSystem%20Component%20Object%20References) (AudioComponents :type Array%20of%20Audio%20Component%20Object%20References))
  ;[1 1 1] Play Shield Start Sounds
  (bind _asbp_player_pawn_zeus (Utilities|Casting|CastToBP_PlayerPawn_Zeus MyTarget))
  (Utilities|FlowControl|Delay 0 1.6)
  (Audio|SpawnSoundAttached "/SunRoseZeusGameplay/Sounds/Boss/SunRose_ZeusBoss_Shield_Start_Cue.SunRose_ZeusBoss_Shield_Start_Cue" (Variables|/Script/Engine.Actor|GetRootComponent _asbp_player_pawn_zeus)))

(fn ConstructionScript)
