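; Gameplay cue notify actor for the target tell of Zeus's leap, currently a placeholder with no visual or audio logic in its graphs.
; Its OnBurst override only prints a debug "OnActive" string and returns false.
; BeginPlay, overlap, Tick and the construction script are empty.
; Despite the GCNL name and the "OnActive" text, the implemented hook is a burst callback.
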(event EventBeginPlay)

(event Collision|EventActorBeginOverlap
  (params (OtherActor :type Actor%20Object%20Reference)))

(event EventTick
  (params (DeltaSeconds :type Float%20%28single-precision%29)))

(fn OnBurst
  (params (MyTarget :type Actor%20Object%20Reference) (Parameters :type Gameplay%20Cue%20Parameters%20Structure%20%28by%20ref%29))
  (Development|PrintString 0 "OnActive")
  (return false))

(fn ConstructionScript)
